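# Material for a shop tile background in the Safflower season data UI, judging by its name and folder.
# The expression graph is empty, so it holds no visual logic of its own here.
# It lists the Default Lit shading model and has automatic usage setting in the editor turned off.
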Material /Safflower_SeasonData/UI/ShopFiles/M_BG_Nexus_Safflower_SeasonData_ShopTile
graph {expressions {}, properties {bAutomaticallySetUsageInEditor=false, usedShadingModels=Default Lit}}
